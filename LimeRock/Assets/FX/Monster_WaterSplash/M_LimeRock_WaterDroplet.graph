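# Translucent unlit, two-sided Niagara sprite material for water droplets in the LimeRock monster water splash effect.
# The green channel of a splash texture parameter tints particle color and, raised to a power and amplified, forms a mask multiplied by particle alpha.
# That mask passes through a depth fade to soften intersections with scene geometry, with power, amplitude and fade distance exposed as Opacity parameters.
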Material /LimeRock/Assets/FX/Monster_WaterSplash/M_LimeRock_WaterDroplet
graph {
 properties {
  bAutomaticallySetUsageInEditor=false
  bUsedWithNiagaraSprites=true
  blendMode=BLEND_Translucent
  shadingModel=MSM_Unlit
  twoSided=true
  usedShadingModels=Unlit
 }
 nodes [
  n1 = ParticleColor()
  n2 = TextureSampleParameter2D() {parameterName=Splash_Texture, texture=&/T_LimeRock_Splash}
  n3 = Power(Base=@n2.G, Exp=ScalarParameter() {defaultValue=10, group=Opacity, parameterName=Opacity_Power}) {constExponent=15}
  n4 = ScalarParameter() {defaultValue=10, group=Opacity, parameterName=Opacity_Amp}
  DepthFade(
   Opacity=Multiply(A=@n1.A, B=Multiply(A=@n3, B=@n4) {constB=20})
   FadeDistance=ScalarParameter() {defaultValue=3000, group=Opacity, parameterName=DepthFade}
  )
  Multiply(A=@n1.RGB, B=@n2.G)
 ]
}
